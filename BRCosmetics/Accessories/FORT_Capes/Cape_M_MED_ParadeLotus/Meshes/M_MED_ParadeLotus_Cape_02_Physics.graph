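# Physics asset for the ParadeLotus cape cosmetic, defining bodies and joints for cape motion on a medium male character.
# Center, left and right chains of five capsules each hang from master bodies linked to spine_05.
# Chain joints swing mainly on one axis, with little or no twist and side swing.
# Pelvis, spine and arm shapes approximate the wearer's body, joined by fully free angular constraints, as are the cape masters.
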PhysicsAsset /BRCosmetics/Accessories/FORT_Capes/Cape_M_MED_ParadeLotus/Meshes/M_MED_ParadeLotus_Cape_02_Physics
bodies {
 dyn_cape_01 [{shapeType=Capsule, length=7.5, radius=6, shapeName=dyn_cape_01_capsule, center=<8 0 8>, rotation=<rot 90 0 0>}]
 dyn_cape_02 [{shapeType=Capsule, length=7.5, radius=6, shapeName=dyn_cape_01_capsule, center=<8 0 5>, rotation=<rot 90 0 0>}]
 dyn_cape_03 [{shapeType=Capsule, length=7.5, radius=6, shapeName=dyn_cape_01_capsule, center=<8 0 5>, rotation=<rot 90 0 0>}]
 dyn_cape_04 [{shapeType=Capsule, length=7.5, radius=6, shapeName=dyn_cape_01_capsule, center=<8 0 5>, rotation=<rot 90 0 0>}]
 dyn_cape_05 [{shapeType=Capsule, length=7.5, radius=6, shapeName=dyn_cape_01_capsule, center=<8 0 5>, rotation=<rot 90 0 0>}]
 dyn_cape_MSTR [{shapeType=Capsule, length=0.505, radius=6, shapeName=dyn_cape_r_MSTR_capsule, rotation=<rot 0 -90 0>}]
 dyn_cape_l_01 [{shapeType=Capsule, length=7.5, radius=6, shapeName=dyn_cape_01_capsule, center=<8 0 8>, rotation=<rot 90 0 0>}]
 dyn_cape_l_02 [{shapeType=Capsule, length=7.5, radius=6, shapeName=dyn_cape_01_capsule, center=<8 0 5>, rotation=<rot 90 0 0>}]
 dyn_cape_l_03 [{shapeType=Capsule, length=7.5, radius=6, shapeName=dyn_cape_01_capsule, center=<8 0 5>, rotation=<rot 90 0 0>}]
 dyn_cape_l_04 [{shapeType=Capsule, length=7.5, radius=6, shapeName=dyn_cape_01_capsule, center=<8 0 5>, rotation=<rot 90 0 0>}]
 dyn_cape_l_05 [{shapeType=Capsule, length=7.5, radius=6, shapeName=dyn_cape_01_capsule, center=<8 0 5>, rotation=<rot 90 0 0>}]
 dyn_cape_l_MSTR [{shapeType=Capsule, length=0.505, radius=0.505, shapeName=dyn_cape_r_MSTR_capsule, rotation=<rot 0 -90 0>}]
 dyn_cape_r_01 [{shapeType=Capsule, length=7.5, radius=6, shapeName=dyn_cape_01_capsule, center=<8 0 8>, rotation=<rot 90 0 0>}]
 dyn_cape_r_02 [{shapeType=Capsule, length=7.5, radius=6, shapeName=dyn_cape_01_capsule, center=<8 0 5>, rotation=<rot 90 0 0>}]
 dyn_cape_r_03 [{shapeType=Capsule, length=7.5, radius=6, shapeName=dyn_cape_01_capsule, center=<8 0 5>, rotation=<rot 90 0 0>}]
 dyn_cape_r_04 [{shapeType=Capsule, length=7.5, radius=6, shapeName=dyn_cape_01_capsule, center=<8 0 5>, rotation=<rot 90 0 0>}]
 dyn_cape_r_05 [{shapeType=Capsule, length=7.5, radius=6, shapeName=dyn_cape_01_capsule, center=<8 0 5>, rotation=<rot 90 0 0>}]
 dyn_cape_r_MSTR [{shapeType=Capsule, length=0.505, radius=0.505, shapeName=dyn_cape_r_MSTR_capsule, rotation=<rot 0 -90 0>}]
 lowerarm_l [{shapeType=Capsule, length=18.1455, radius=9.827, shapeName=lowerarm_l_capsule, center=<5.6061 0 0.109>, rotation=<rot 0 -90 -87.2397>}]
 lowerarm_r [{shapeType=Capsule, length=18.1455, radius=9.8794, shapeName=lowerarm_r_capsule, center=<-5.6061 0 -0.109>, rotation=<rot 0 -90 92.7603>}]
 pelvis [{shapeType=Capsule, length=34.0387, radius=22.3325, center=<-18.6191 -4.7982 0>}]
 spine_01 [{shapeType=Capsule, length=7.9438, radius=13.5586, center=<-1.7965 -1.3367 0>}]
 spine_02 [{shapeType=Capsule, length=7.3264, radius=12.6474, center=<-1.711 0.2698 0.0001>}]
 spine_03 [{shapeType=Capsule, length=8.2107, radius=13.7031, center=<0.1445 0.1222 0.0001>}]
 spine_04 [{shapeType=Capsule, length=29.718, radius=15.767, center=<5.4899 2.9127 0.0002>}]
 spine_05 [{shapeType=Sphere, radius=16.1415, center=<-0.4501 2.6408 0>}]
 upperarm_l [{
   shapeType=Capsule
   length=19
   radius=8.8857
   shapeName=upperarm_l_capsule
   center=<11.8666 -0.7476 0.9566>
   rotation=<rot -0.0132 -89.9882 99.6113>
  }]
 upperarm_r [{
   shapeType=Capsule
   length=19
   radius=7.2213
   shapeName=upperarm_r_capsule
   center=<-11.8666 0.7476 -0.9566>
   rotation=<rot -0.0132 -89.9882 -80.3887>
  }]
}
constraints [
 {
  bone1Name=spine_01
  bone2Name=pelvis
  swing1LimitDegrees=45
  swing1Motion=Free
  swing2LimitDegrees=45
  swing2Motion=Free
  twistLimitDegrees=45
  twistMotion=Free
 }
 {
  bone1Name=spine_02
  bone2Name=spine_01
  swing1LimitDegrees=45
  swing1Motion=Free
  swing2LimitDegrees=45
  swing2Motion=Free
  twistLimitDegrees=45
  twistMotion=Free
 }
 {
  bone1Name=spine_03
  bone2Name=spine_02
  swing1LimitDegrees=45
  swing1Motion=Free
  swing2LimitDegrees=45
  swing2Motion=Free
  twistLimitDegrees=45
  twistMotion=Free
 }
 {
  bone1Name=spine_04
  bone2Name=spine_03
  swing1LimitDegrees=45
  swing1Motion=Free
  swing2LimitDegrees=45
  swing2Motion=Free
  twistLimitDegrees=45
  twistMotion=Free
 }
 {
  bone1Name=spine_05
  bone2Name=spine_04
  swing1LimitDegrees=45
  swing1Motion=Free
  swing2LimitDegrees=45
  swing2Motion=Free
  twistLimitDegrees=45
  twistMotion=Free
 }
 {
  bone1Name=dyn_cape_MSTR
  bone2Name=spine_05
  swing1LimitDegrees=45
  swing1Motion=Free
  swing2LimitDegrees=45
  swing2Motion=Free
  twistLimitDegrees=45
  twistMotion=Free
 }
 {
  bone1Name=dyn_cape_r_MSTR
  bone2Name=spine_05
  swing1LimitDegrees=45
  swing1Motion=Free
  swing2LimitDegrees=45
  swing2Motion=Free
  twistLimitDegrees=45
  twistMotion=Free
 }
 {
  bone1Name=dyn_cape_l_MSTR
  bone2Name=spine_05
  swing1LimitDegrees=45
  swing1Motion=Free
  swing2LimitDegrees=45
  swing2Motion=Free
  twistLimitDegrees=45
  twistMotion=Free
 }
 {
  bone1Name=upperarm_l
  bone2Name=spine_05
  swing1LimitDegrees=45
  swing1Motion=Free
  swing2LimitDegrees=45
  swing2Motion=Free
  twistLimitDegrees=45
  twistMotion=Free
 }
 {
  bone1Name=upperarm_r
  bone2Name=spine_05
  swing1LimitDegrees=45
  swing1Motion=Free
  swing2LimitDegrees=45
  swing2Motion=Free
  twistLimitDegrees=45
  twistMotion=Free
 }
 {bone1Name=dyn_cape_01, bone2Name=dyn_cape_MSTR, swing1LimitDegrees=0, swing2LimitDegrees=20, twistLimitDegrees=0}
 {bone1Name=dyn_cape_02, bone2Name=dyn_cape_01, swing1LimitDegrees=3, swing2LimitDegrees=20, twistLimitDegrees=2}
 {bone1Name=dyn_cape_03, bone2Name=dyn_cape_02, swing1LimitDegrees=3, swing2LimitDegrees=20, twistLimitDegrees=2}
 {bone1Name=dyn_cape_04, bone2Name=dyn_cape_03, swing1LimitDegrees=3, swing2LimitDegrees=20, twistLimitDegrees=2}
 {bone1Name=dyn_cape_05, bone2Name=dyn_cape_04, swing1LimitDegrees=3, swing2LimitDegrees=20, twistLimitDegrees=2}
 {bone1Name=dyn_cape_l_01, bone2Name=dyn_cape_l_MSTR, swing1LimitDegrees=0, swing2LimitDegrees=20, twistLimitDegrees=0}
 {bone1Name=dyn_cape_l_02, bone2Name=dyn_cape_l_01, swing1LimitDegrees=3, swing2LimitDegrees=20, twistLimitDegrees=2}
 {bone1Name=dyn_cape_l_03, bone2Name=dyn_cape_l_02, swing1LimitDegrees=3, swing2LimitDegrees=20, twistLimitDegrees=2}
 {bone1Name=dyn_cape_l_04, bone2Name=dyn_cape_l_03, swing1LimitDegrees=3, swing2LimitDegrees=20, twistLimitDegrees=2}
 {bone1Name=dyn_cape_l_05, bone2Name=dyn_cape_l_04, swing1LimitDegrees=3, swing2LimitDegrees=20, twistLimitDegrees=2}
 {bone1Name=dyn_cape_r_01, bone2Name=dyn_cape_r_MSTR, swing1LimitDegrees=0, swing2LimitDegrees=20, twistLimitDegrees=0}
 {bone1Name=dyn_cape_r_02, bone2Name=dyn_cape_r_01, swing1LimitDegrees=3, swing2LimitDegrees=20, twistLimitDegrees=2}
 {bone1Name=dyn_cape_r_03, bone2Name=dyn_cape_r_02, swing1LimitDegrees=3, swing2LimitDegrees=20, twistLimitDegrees=2}
 {bone1Name=dyn_cape_r_04, bone2Name=dyn_cape_r_03, swing1LimitDegrees=3, swing2LimitDegrees=20, twistLimitDegrees=2}
 {bone1Name=dyn_cape_r_05, bone2Name=dyn_cape_r_04, swing1LimitDegrees=3, swing2LimitDegrees=20, twistLimitDegrees=2}
 {
  bone1Name=lowerarm_r
  bone2Name=upperarm_r
  swing1LimitDegrees=45
  swing1Motion=Free
  swing2LimitDegrees=45
  swing2Motion=Free
  twistLimitDegrees=45
  twistMotion=Free
 }
 {
  bone1Name=lowerarm_l
  bone2Name=upperarm_l
  swing1LimitDegrees=45
  swing1Motion=Free
  swing2LimitDegrees=45
  swing2Motion=Free
  twistLimitDegrees=45
  twistMotion=Free
 }
]
