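# Physics asset for the ParadeLotus cape cosmetic; its name indicates it supplies the body collision for the cape's cloth simulation.
# It defines bodies on the pelvis, the five spine bones, and the left and right clavicles, upper arms and lower arms, covering torso and arms only.
# No constraints join the bodies, so it is not set up as a jointed ragdoll.
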PhysicsAsset /BRCosmetics/Accessories/FORT_Capes/Cape_M_MED_ParadeLotus/Meshes/M_MED_ParadeLotus_Cloth_Collisions
bodies {
 clavicle_l []
 clavicle_r []
 lowerarm_l []
 lowerarm_r []
 pelvis []
 spine_01 []
 spine_02 []
 spine_03 []
 spine_04 []
 spine_05 []
 upperarm_l []
 upperarm_r []
}
constraints []
